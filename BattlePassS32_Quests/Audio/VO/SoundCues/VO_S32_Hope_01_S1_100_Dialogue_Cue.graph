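# Sound cue for a Season 32 Battle Pass quest voice-over line from the Hope dialogue set.
# A single dialogue player node plays the line's dialogue wave in the audio-log speaker context with no targets.
# It uses audio-log attenuation and the narrator VO sound class.
# A pre-effect send feeds a quest VO envelope-follower audio bus, exposing the voice's loudness for modulation.
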SoundCue /BattlePassS32_Quests/Audio/VO/SoundCues/VO_S32_Hope_01_S1_100_Dialogue_Cue
root=DialoguePlayer_0
nodes [{
  id=DialoguePlayer_0
  properties {dialogueWaveParameter {
    dialogueWave=/BattlePassS32_Quests/Audio/VO/DialogueWaves/VO_S32_Hope_01_S1_100_Dialogue
    context {speaker=/Game/Sounds/Fort_Audio_Blueprints/Utilities/EditorUtilityWidgets/VO_Utilities/Assets/AudioLogVO_DV, targets []}
   }}
 }]
properties {
 attenuationSettings=/Game/Sounds/Fort_Audio_Blueprints/Utilities/EditorUtilityWidgets/VO_Utilities/Assets/AS_Fort_AudioLogs
 maxDistance=2097152
 soundClassObject=/Game/Sounds/FortSoundClassesAndModes/Fort_VO_Narrator
 volumeMultiplier=1
 preEffectBusSends [{
   audioBus=/Game/Sounds/Modulation/EnvelopeFollowers/BR_Boss/AB_Quest_VO_EnvFollow
   maxSendDistance=8000
   minSendDistance=0
   soundSourceBus=null
  }]
}
